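# Translucent unlit Niagara sprite material for pixelated smoke particles, animated from a 4x4 flipbook texture.
# The Animation dynamic parameter sets the flipbook phase that drives the UVs of the smoke texture.
# The texture's green channel multiplies particle color and feeds the shared MF_FX_EmissiveAdjustments function for emissive.
# The red channel serves as the alpha mask, multiplied by particle alpha.
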Material /Game/Effects/Niagara/MasterMaterials/Pixelated/M_PixelatedSmoke
graph {
 properties {
  bUsedWithNanite=true
  bUsedWithNiagaraSprites=true
  bUsedWithParticleSprites=true
  blendMode=BLEND_Translucent
  pixelDepthOffsetMode=PDOM_Legacy
  shadingModel=MSM_Unlit
  translucencyLightingMode=TLM_VolumetricPerVertexDirectional
  translucentSelfShadowDensityScale=1
  usedShadingModels=Unlit
  translucentMultipleScatteringExtinction=<rgba 1 1 1 1>
 }
 nodes [
  n1 = ParticleColor()
  n2 = Constant() {r=4}
  n3 = TextureSample(UVs=MaterialFunctionCall(
    "Animation  Phase (0-1)"=DynamicParameter() {paramNames [Animation, Unused, Emissive, NormalMultiplier]}.Animation
    Number of Rows=@n2
    Number of Columns=@n2
    UVs=TextureCoordinate()
   ) {materialFunction=&FlipBook}.UVs) {texture=/Game/Effects/Niagara/MasterMaterials/Pixelated/Textures/T_PixelatedSmoke}
  MaterialFunctionCall(Emissive=Multiply(A=@n1.RGB, B=NamedRerouteDeclaration(None=@n3.G) {name=Emissive_Flipbook, nodeColor=<rgba 1 1 0.3516 1>})) {materialFunction=&MF_FX_EmissiveAdjustments}
  Multiply(A=@n1.A, B=NamedRerouteDeclaration(None=@n3.R) {name=Alpha_Flipbook, nodeColor=<rgba 1 0 0.3516 1>})
 ]
}
